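# Translucent UI material for a Collections track, drawn as a two-color horizontal progress bar with angled, trimmed edges.
# Progress scaled by Anim sets a sharp threshold along normalized UV X that blends RGB1 (unfilled) to RGB2 (filled).
# Rotated UVs with offset and sharpness parameters form top, left and right edge masks, the sides using 9-slice UVs.
# Their product shapes opacity, with RGB1 alpha setting the unfilled portion's visibility.
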Material /FortniteCoreUI/Athena/Collections/Art/M_Ui_Collection_Track
graph {
 properties {
  bUsedWithStaticMesh=false
  blendMode=BLEND_Translucent
  materialDomain=MD_UI
  pixelDepthOffsetMode=PDOM_Legacy
  usedShadingModels=Default Lit
 }
 nodes [
  n1 = VectorParameter() {parameterName=RGB1, defaultValue=<rgba 0 0 0.136 1>}
  n2 = ScalarParameter() {defaultValue=0.75, parameterName=Progress, sliderMax=1}
  n3 = ScalarParameter() {defaultValue=1, parameterName=Anim, sliderMax=1}
  n4 = Subtract(
   A=ComponentMask(None=MaterialFunctionCall() {materialFunction=&GetUserInterfaceUV}.Normalized UV) {r=true}
   B=LinearInterpolate(Alpha=LinearInterpolate(B=@n2, Alpha=@n3)) {constA=-0.001}
  )
  n5 = ScalarParameter() {defaultValue=128, group=Border sharpness, parameterName=ProgressSharpness}
  n6 = OneMinus(None=Saturate(None=Multiply(A=@n4, B=@n5)))
  LinearInterpolate(A=@n1.RGB, B=VectorParameter() {parameterName=RGB2, defaultValue=<rgba 0 1 1 1>}.RGB, Alpha=@n6)
  n7 = ComponentMask(None=Rotator(
    Coordinate=MaterialFunctionCall() {materialFunction=&GetUserInterfaceUV}.Normalized UV
    Time=ScalarParameter() {defaultValue=0.05, group=Border Angles, parameterName=Top-Angle}
   ) {speed=6.2832}) {g=true}
  n8 = ScalarParameter() {defaultValue=0.1857, group=Border_Range, parameterName=Top-RangeOffset, sliderMax=1}
  n9 = Saturate(None=Multiply(A=Subtract(A=@n7, B=@n8), B=ScalarParameter() {defaultValue=8, group=Border sharpness, parameterName=HorizontalSharpness}))
  n10 = ComponentMask(None=Rotator(
    Coordinate=MaterialFunctionCall() {materialFunction=&GetUserInterfaceUV}.9-Slice UV
    Time=ScalarParameter() {defaultValue=-0.01, group=Border Angles, parameterName=side-Angle}
   ) {speed=6.2832}) {r=true}
  n11 = ScalarParameter() {defaultValue=0.009, group=Border_Range, parameterName=Left-RangeOffset, sliderMax=1}
  n12 = ScalarParameter() {defaultValue=128, group=Border sharpness, parameterName=VerticalSharpness}
  n13 = Saturate(None=Multiply(A=Subtract(A=@n10, B=@n11), B=@n12))
  n14 = ScalarParameter() {defaultValue=0.985, group=Border_Range, parameterName=Right-RangeOffset, sliderMax=1}
  n15 = Saturate(None=Multiply(A=Subtract(A=@n10, B=@n14), B=@n12))
  Multiply(A=DotProduct(A=DotProduct(A=@n9, B=@n13), B=OneMinus(None=@n15)), B=Saturate(None=Add(A=@n6, B=@n1.A)))
 ]
}
